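# GPU Niagara system for several pigeon flocks in UtopiaCity that rest until activated, then take flight.
# Leaders spawn around a ring, latch active on camera distance, then rise; inactive particles do not age.
# Followers read Leaders through a particle attribute reader and swirl around them with vortex, wind and drag.
# A separate SceneDepth emitter scatters points using baked depth and rejection textures, snaps them to distance-field surfaces and culls by normal.
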NiagaraSystem /Game/Effects/Fort_Effects/Effects/Environment/Hermes/UtopiaCity/Pigeons/NS_UtopiaCity_Pigeons_Multiflock
dependencies {
 usedDataInterfaces [
  /Script/Niagara.NiagaraDataInterfaceTexture
  &NiagaraDataInterfaceCurve
  &NiagaraDataInterfaceCamera
  &NiagaraDataInterfaceParticleRead
  &NiagaraDataInterfaceCollisionQuery
  &NiagaraDataInterfaceLandscape
 ]
 usedDynamicInputs []
 usedModules [
  &self:SystemSpawnScript.SystemScriptSource.SystemScriptGraph.NiagaraNodeAssignment_0.SetVariables_#1
  &SystemState
 ]
}
emitters {
 Followers {
  simTarget=GPUComputeSim
  emitterSpawnScript [/SetVariables(Emitter.AttributeReader=DataInterface() {propertyValues {EmitterBinding {bindingMode=Other, emitterName=Leaders}}}) {isSetParameters=true}]
  emitterUpdateScript [
   &EmitterState(
    Inactive Response=::NewEnumerator1
    Life Cycle Mode=::NewEnumerator1
    Loop Behavior=::NewEnumerator1
    Loop Duration Mode=::NewEnumerator1
    UseLoopDelay=false
   )
   &SpawnBurst_Instantaneous(Spawn Count=200, Use Spawn Probability=false, UseLoopCountLimit=true)
  ]
  particleSpawnScript [
   &InitializeParticle(
    Color=<rgba 1 0 0.5895 10>
    Lifetime=0.5
    Mass Max=100
    Mass Min=0.1
    Mass Mode=::NewEnumerator3
    Mesh Scale Mode=::NewEnumerator1
    Mesh Uniform Scale Max=1
    Mesh Uniform Scale Min=0.5
    Sprite Size Mode=::NewEnumerator3
    Uniform Sprite Size=25
    UsePositionOffset=false
    Write Material Random=true
   )
   &self:ScratchModule_01(Attribute Reader:NiagaraDataInterfaceParticleRead=$Emitter.AttributeReader) {module=ScratchModule_01001}
   &CameraOffset(Camera Offset Amount=100)
   &DynamicMaterialParameters(
    Index 0 Param 1=0.25
    Index 0 Param 2=&RandomRangeFloat(Minimum=1, Maximum=1.2)
    Index 0 Param 3=0.25
    Index 0 Param 4=0.1
    Index 1 Param 1=1
    Index 1 Param 2=&RandomRangeFloat(Minimum=8, Maximum=10)
    Index 1 Param 3=2
    Index 2 Param 1=0.1
    Index 2 Param 2=1
    Index 2 Param 3=2
    Index 2 Param 4=0.25
    Index 3 Param 1=&RandomRangeFloat(Minimum=1.5, Maximum=2)
    Index 3 Param 4=1
    Module.Param0WriteEnabled=true
    Module.Param0WriteEnabled1=true
    Module.Param0WriteEnabled2=true
    Module.Param0WriteEnabled3=true
    Module.Param1WriteEnabled=true
    Module.Param1WriteEnabled1=true
    Module.Param1WriteEnabled2=true
    Module.Param1WriteEnabled3=true
    Module.Param2WriteEnabled=true
    Module.Param2WriteEnabled1=true
    Module.Param2WriteEnabled2=true
    Module.Param2WriteEnabled3=true
    Module.Param3WriteEnabled=true
    Module.Param3WriteEnabled1=true
    Module.Param3WriteEnabled2=true
    Module.Param3WriteEnabled3=true
    Write Parameter Index 1=true
    Write Parameter Index 2=true
    Write Parameter Index 3=true
   )
  ]
  particleUpdateScript [
   &self:ScratchModule_02(AttributeReader=DataInterface() {propertyValues {EmitterBinding {bindingMode=Other, emitterName=Leaders}}}) {module=ScratchModule_02}
   /SetVariables(Particles.DistanceToLeader:NiagaraFloat=&DistanceBetweenPositions(A=$Particles.LeaderPosition, B=$Particles.Position)) {isSetParameters=true}
   &ParticleState(DeltaTime:NiagaraFloat=&MakeCustomFloatFromBool(Bool=$Particles.IsActive, True Float=$Engine.DeltaTime, False Float=0))
   &PointAttractionForce(
    Attraction Radius:NiagaraFloat=&Multiply_Float(A=$Particles.FlockRadius, B=2)
    Attraction Strength=&MakeCustomFloatFromBool(
     Bool=$Particles.IsActive
     True Float=&Multiply_Float(A=&RandomRangeFloat(Minimum=0.1, Maximum=1), B=250)
     False Float=0
    )
    Attractor Position:NiagaraPosition=$Particles.LeaderPosition
    Kill Within Radius=false
    Module.Use Falloff=true
    UseAttractorOffset=false
   ) {enabled=false}
   &VortexForce(
    Origin Pull Amount:NiagaraFloat=&MakeCustomFloatFromBool(Bool=$Particles.IsActive, True Float=1000, False Float=0)
    UseOriginOffset=false
    Vortex Axis=&Lerp_Vector(
     A=<-0.15 0.15 0>
     B=<0.25 0 0.25>
     Alpha=&FloatFromCurve(
      FloatCurve=DataInterface() {propertyValues {Curve {keys [
          {interpMode=RCIM_Cubic, tangentMode=RCTM_User, time=0, value=0}
          {interpMode=RCIM_Cubic, tangentMode=RCTM_User, time=0.5, value=1}
          {tangentMode=RCTM_User, time=1, value=0}
         ]}}}
      CurveIndex=$System.NormalizedLoopAge
     )
    )
    Vortex Force Amount=&MakeCustomFloatFromBool(Bool=$Particles.IsActive, True Float=10000, False Float=0)
    Vortex Origin:NiagaraPosition=$Particles.LeaderPosition
   )
   &WindForce(
    Frequency=0.5
    Friction Mode:ENiagaraWindFrictionMode=::NewEnumerator0
    Scale:NiagaraFloat=&RandomRangeFloat(Minimum=0, Maximum=0.1)
    Wind Speed=<0.5 0 1>
    Wind Speed Scale=&MakeCustomFloatFromBool(Bool=$Particles.IsActive, True Float=20000, False Float=0)
   )
   &Drag(
    Drag=&MakeCustomFloatFromBool(Bool=$Particles.IsActive, True Float=1, False Float=0)
    Ignore Mass=false
    Use Linear Drag=true
    Use Rotational Drag=true
   )
   &SolveForcesAndVelocity(Clamp Velocity=true, Limit Acceleration=false, Speed Limit=300)
  ]
  rendererClasses [&NiagaraSpriteRendererProperties, &NiagaraMeshRendererProperties]
  renderers [
   {rendererClass=&NiagaraSpriteRendererProperties}
   {rendererClass=&NiagaraMeshRendererProperties, rendererIndex=1}
  ]
 }
 Leaders {
  simTarget=GPUComputeSim
  emitterUpdateScript [
   &EmitterState()
   /SetVariables() {isSetParameters=true}
   /SetVariables(Emitter.CameraPosition:NiagaraPosition=&ReturnCameraPosition(Coordinate Space:ENiagaraCoordinateSpace=::World, Camera Query:NiagaraDataInterfaceCamera=DataInterface())) {isSetParameters=true}
   /SetVariables(Emitter.FlockCount=10) {isSetParameters=true}
   &SpawnBurst_Instantaneous(Spawn Count=$Emitter.FlockCount, Use Spawn Probability=false, UseLoopCountLimit=true)
  ]
  particleSpawnScript [
   /SetVariables() {isSetParameters=true}
   /SetVariables(Particles.FlockRadius:NiagaraFloat=&Multiply_Float(A=&RandomRangeFloat(Minimum=0.25, Maximum=1), B=1000)) {isSetParameters=true}
   &InitializeParticle(
    Color Mode=::NewEnumerator0
    Lifetime=7
    Mesh Scale Mode=::NewEnumerator3
    Mesh Uniform Scale=10
    Sprite Size Mode=::NewEnumerator3
    Uniform Sprite Size=$Particles.FlockRadius
    UsePositionOffset=false
    Write Material Random=true
   )
   &ShapeLocation(
    MaskSpawnGroup=false
    Ring / Disc Distribution Mode:ENiagaraSphereDistributionMode=::NewEnumerator3
    Ring Radius=5000
    Shape Primitive=::NewEnumerator4
    UseOwnerScale=true
   )
   /SetVariables(Particles.RestPosition:NiagaraPosition=$Particles.Position) {isSetParameters=true}
   /SetVariables(Particles.DEBUGFlockRadius:Vector3f=&VectorFromFloat(Value=$Particles.FlockRadius)) {isSetParameters=true}
  ]
  particleUpdateScript [
   &ParticleState(DeltaTime:NiagaraFloat=&MakeCustomFloatFromBool(Bool=$Particles.IsActive, True Float=$Engine.DeltaTime, False Float=0))
   /SetVariables(Particles.IsActive:NiagaraBool=&Boolean_OrOperation(
     Bool1=$Particles.IsActive
     Bool2=&SetBoolByFloatComparison(
      A=&DistanceBetweenPositions(A=$Particles.Position, B=$Emitter.CameraPosition)
      B=4000
      Comparison Type=::NewEnumerator6
     )
    )) {isSetParameters=true}
   &LinearForce(Force=&Multiply_VectorByFloat(
     Vector=<0.5 0 1>
     Float=&MakeCustomFloatFromBool(Bool=$Particles.IsActive, True Float=1000, False Float=0)
    ))
   &Drag(
    Drag=&MakeCustomFloatFromBool(Bool=$Particles.IsActive, True Float=3, False Float=0)
    Use Linear Drag=true
    Use Rotational Drag=true
   )
   &SolveForcesAndVelocity(Clamp Velocity=false, Limit Acceleration=false)
  ]
  rendererClasses [&NiagaraMeshRendererProperties]
  renderers [
   {rendererClass=&NiagaraMeshRendererProperties, rendererIndex=0}
   {rendererClass=&NiagaraMeshRendererProperties, rendererIndex=1}
  ]
 }
 SceneDepth {
  simTarget=GPUComputeSim
  emitterUpdateScript [
   &EmitterState()
   &SpawnBurst_Instantaneous(Spawn Count=5000, Use Spawn Probability=false, UseLoopCountLimit=true)
  ]
  particleSpawnScript [
   &InitializeParticle(
    Color Mode=::NewEnumerator0
    Lifetime=5
    Mesh Scale Mode=::NewEnumerator3
    Mesh Uniform Scale=20
    Sprite Size Mode=::NewEnumerator3
    Uniform Sprite Size=100
    UsePositionOffset=false
    Write Material Random=true
   )
   /Niagara/Modules/Update/Textures/SampleTexture(
    Apply Sampled RGBA to Particle Color=false
    Rotation Angle=90
    Rotation Angle Type:ENiagara_AngleInput=::NewEnumerator0
    Rotation Axis:Vector3f=<0 0 1>
    Rotation Coordinate Space:ENiagaraCoordinateSpace=::World
    Rotation Mode:ENiagara_RotationMode=::NewEnumerator1
    Texture=DataInterface() {propertyValues {
      Texture=/Game/Effects/Fort_Effects/Effects/Environment/Hermes/UtopiaCity/Pigeons/Textures/T_UtopiaCity_SceneDepth
      TextureUserParameter {parameter {name=None, typeDefHandle {registeredTypeIndex=2}, varData []}}
     }}
    Transform UV Into Position=true
    UV:Vector2f=&RandomRangeVector2D(Minimum=<0 0>, Maximum=<1 1>)
    UV Scale:Vector2f=&Vector2DFromFloat(Value=$System.CamOrthoWidth)
   )
   &SampleTexture(
    Apply Sampled RGBA to Particle Color=false
    "Comparison Value (A)":ENiagara_ImportanceRejectionMode=::NewEnumerator3
    "Comparison Value (B)"=1
    Enable Sample Rejection=true
    Texture=DataInterface() {propertyValues {
      Texture=/Game/Effects/Fort_Effects/Effects/Environment/Hermes/UtopiaCity/Pigeons/Textures/T_UtopiaCity_Rejection
      TextureUserParameter {parameter {name=None, typeDefHandle {registeredTypeIndex=2}, varData []}}
     }}
    UV:Vector2f=$Particles.SampleTexture.SamplerUV
   ) {module=SampleTexture001}
   /SetVariables(Particles.Position=&AddVectorToPosition(
     Position=$Particles.Position
     Vector=&MakeVector(
      X=0
      Y=0
      Z=&RemapRange(Input Value To Remap=$Particles.SampleTexture.SampledR, Input Min=0.595, Input Max=0.154, Output Min=-50, Output Max=4312)
     )
    )) {isSetParameters=true}
   /Niagara/Modules/Math/MoveToNearestDistanceFieldSurface_GPU(Offset Intrinsic Position By Radius=true, Radius=50, bKillByDistance=true)
   /SetVariables(Particles.SurfaceNormal:Vector3f=&Multiply_VectorByFloat(
     Vector=&NormalizeVector(
      Vector To Normalize=$Output.MoveToNearestDistanceFieldSurface_GPU.GlobalDistanceFieldAveragedNormal
      Safe Normalize=true
      Safe Fallback Threshold=0
      Safe Fallback Vector=<0 0 0.001>
     )
     Float=-1
    )) {isSetParameters=true}
   &KillParticles(Kill Particles=&SetBoolByFloatComparison(A=&DotProduct(Vector A=$Particles.SurfaceNormal, Vector B=<0 0 1>), B=0.9, Comparison Type=::NewEnumerator6))
  ]
  particleUpdateScript [&ParticleState(Kill Particles When Lifetime Has Elapsed=false)]
  rendererClasses [&NiagaraSpriteRendererProperties, &NiagaraMeshRendererProperties]
  renderers [
   {rendererClass=&NiagaraSpriteRendererProperties}
   {rendererClass=&NiagaraMeshRendererProperties, rendererIndex=1}
  ]
 }
}
summary {
 systemName=NS_UtopiaCity_Pigeons_Multiflock
 emitters [
  {
   emitterName=SceneDepth
   simTarget=GPUComputeSim
   rendererClasses [&NiagaraSpriteRendererProperties, &NiagaraMeshRendererProperties]
  }
  {emitterName=Leaders, simTarget=GPUComputeSim, rendererClasses [&NiagaraMeshRendererProperties]}
  {
   emitterName=Followers
   simTarget=GPUComputeSim
   rendererClasses [&NiagaraSpriteRendererProperties, &NiagaraMeshRendererProperties]
  }
 ]
 userVariables []
}
